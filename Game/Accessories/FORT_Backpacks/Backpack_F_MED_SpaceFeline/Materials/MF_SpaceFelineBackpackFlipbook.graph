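# Material function for the Space Feline backpack that overlays flipbook robot eyes and a mouth onto an input material's Base Color and Emissive.
# Happy, blink, sad and blank pose scalars select each feature's sub-UV frame or hide it, defaulting to neutral.
# Each feature has adjustable UV offset and scale.
# A static switch chooses packed masks tinted by color parameters or the textures' own color and alpha, with an Alpha input channel limiting coverage.
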MaterialFunction /Game/Accessories/FORT_Backpacks/Backpack_F_MED_SpaceFeline/Materials/MF_SpaceFelineBackpackFlipbook
graph {nodes [
  n1 = GetMaterialAttributes(None=FunctionInput() {inputName=Material, inputType=FunctionInput_MaterialAttributes}) {attributeGetTypes [#1, #2]}
  n2 = NamedRerouteUsage()
  n3 = NamedRerouteUsage()
  n4 = GetMaterialAttributes(None=SetMaterialAttributes!(
    MaterialAttributes=@n1.MaterialAttributes
    BaseColor=LinearInterpolate(A=@n1.BaseColor, B=@n2, Alpha=@n3)
    EmissiveColor=LinearInterpolate(A=@n1.EmissiveColor, B=@n2, Alpha=@n3)
   ) {attributeSetTypes [#1, #2]}) {attributeGetTypes [#1, #2]}
  n5 = NamedRerouteUsage()
  n6 = NamedRerouteUsage()
  FunctionOutput(None=SetMaterialAttributes(
    MaterialAttributes=@n4.MaterialAttributes
    BaseColor=LinearInterpolate(A=@n4.BaseColor, B=@n5, Alpha=@n6)
    EmissiveColor=LinearInterpolate(A=@n4.EmissiveColor, B=@n5, Alpha=@n6)
   ) {attributeSetTypes [#1, #2]})
  NamedRerouteDeclaration(None=ScalarParameter() {defaultValue=3, group=EyeFlipbook, parameterName=Default Open Mouth Netural Frame}) {name=DefaultMouth, nodeColor=<rgba 1 0.5625 0 1>}
  n7 = ScalarParameter() {group=MouthFlipbook, parameterName=Mouth UV Offset X}
  n8 = ScalarParameter() {group=MouthFlipbook, parameterName=Mouth UV Offset Y}
  n9 = MaterialFunctionCall(
   UVs=Add(A=NamedRerouteUsage(), B=AppendVector(A=@n7, B=@n8))
   Texture Scale=ScalarParameter() {defaultValue=1, group=MouthFlipbook, parameterName=Mouth UV Scale}
  ) {materialFunction=&ScaleUVsByCenter}
  NamedRerouteDeclaration(None=@n9.UVs) {name=Mouth UV, nodeColor=<rgba 1 0 0.5625 1>}
  n10 = TextureObjectParameter() {
   group=EyeFlipbook
   parameterName=EyeFlipbook
   texture=/Game/Accessories/FORT_Backpacks/Backpack_F_MED_SpaceFeline/Textures/T_RobotEyeFlipbook
  }
  n11 = ScalarParameter() {defaultValue=2, group=EyeFlipbook, parameterName=Eye Sub Images Count}
  n12 = MaterialFunctionCall!(Texture=@n10, UVs=NamedRerouteUsage(), SubImages=@n11, Frame=NamedRerouteUsage()) {materialFunction=/Game/Weapons/WeaponSkins/MaterialFunction/CustomNumbers_SubUV}
  NamedRerouteDeclaration(None=StaticSwitchParameter(True=NamedRerouteUsage(), False=@n12.RGB) {defaultValue=true, group=FlipBook, parameterName=UseFlipbookAsMask}) {name=Eye Color, nodeColor=<rgba 0 1 0.8828 1>}
  n13 = ScalarParameter() {defaultValue=0.1, group=EyeFlipbook, parameterName=THreshold_Eye}
  NamedRerouteDeclaration(None=@n13) {name=ThresholdEye, nodeColor=<rgba 0 0.3672 1 1>}
  n14 = TextureObjectParameter() {
   group=EyeFlipbook
   parameterName=EyeFlipbook_Mask
   samplerType=SAMPLERTYPE_Masks
   texture=/Game/Accessories/FORT_Backpacks/Backpack_F_MED_SpaceFeline/Textures/T_EyeFlipBook_Mask
  }
  n15 = MaterialFunctionCall(Texture=@n14, UVs=NamedRerouteUsage(), SubImages=@n11, Frame=NamedRerouteUsage()) {materialFunction=&CustomNumbers_SubUV}
  n16 = Multiply(
   A=StaticSwitchParameter(True=ComponentMask(None=@n15.RGB) {g=true}, False=@n12.Alpha) {defaultValue=true, group=FlipBook, parameterName=UseFlipbookAsMask}
   B=NamedRerouteUsage()
  )
  n17 = Step(Y=NamedRerouteUsage(), X=ScalarParameter() {group=EyeFlipbook, parameterName=Blank_Eye_Pose}) {constY=0.1}
  NamedRerouteDeclaration(None=Multiply(A=NamedRerouteUsage(), B=LinearInterpolate(A=@n16, Alpha=@n17) {constB=0})) {name=Eye Alpha, nodeColor=<rgba 0.3828 1 0 1>}
  NamedRerouteDeclaration(None=LinearInterpolate(
    A=TextureCoordinate()
    B=TextureCoordinate() {coordinateIndex=1}
    Alpha=ScalarParameter() {defaultValue=1, group=MouthFlipbook, parameterName=UseUVChannel2}
   )) {name=UVcoordOrg, nodeColor=<rgba 0 1 0.4844 1>}
  n18 = ScalarParameter() {group=EyeFlipbook, parameterName=Happy_Eye_Pose}
  n19 = ScalarParameter() {defaultValue=1, group=EyeFlipbook, parameterName=Happy Eye Frame}
  n20 = NamedRerouteUsage()
  n21 = ScalarParameter() {group=EyeFlipbook, parameterName=Blink_Eye_Pose}
  n22 = NamedRerouteUsage()
  n23 = If(A=@n21, B=@n13, "A > B"=ScalarParameter() {defaultValue=2, group=EyeFlipbook, parameterName=Blink Eye Frame}, "A == B"=@n22, "A < B"=@n22)
  n24 = ScalarParameter() {group=EyeFlipbook, parameterName=Sad_Eye_Pose}
  n25 = NamedRerouteUsage()
  n26 = If(A=@n24, B=@n13, "A > B"=ScalarParameter() {group=EyeFlipbook, parameterName=Sad Eye Frame}, "A == B"=@n25, "A < B"=@n25)
  n27 = Step(Y=NamedRerouteUsage(), X=@n24) {constY=0.1}
  n28 = NamedRerouteUsage()
  n29 = Max(A=@n24, B=@n21)
  NamedRerouteDeclaration(None=LinearInterpolate(
    A=If(A=@n18, B=@n13, "A > B"=@n19, "A == B"=@n20, "A < B"=@n20)
    B=LinearInterpolate(A=@n23, B=@n26, Alpha=@n27)
    Alpha=Step(Y=@n28, X=@n29) {constY=0.1}
   )) {name=EyeFrame, nodeColor=<rgba 0.2187 1 0 1>}
  n30 = VectorParameter() {group=FlipBook, parameterName=OuterColor, defaultValue=<rgba 0 0.4793 0.7605 1>}
  NamedRerouteDeclaration(None=LinearInterpolate(
    A=@n30.RGB
    B=VectorParameter() {group=FlipBook, parameterName=InnerColor, defaultValue=<rgba 1 1 1 1>}.RGB
    Alpha=ComponentMask(None=@n15.RGB) {r=true}
   )) {name=EyeColorMasked, nodeColor=<rgba 0.9219 0 1 1>}
  NamedRerouteDeclaration(None=@n30.RGB) {name=EyeOuterColor, nodeColor=<rgba 0 0.9297 1 1>}
  n31 = ScalarParameter() {group=MouthFlipbook, parameterName=Happy_Mouth_Pose}
  n32 = ScalarParameter() {defaultValue=0.1, group=MouthFlipbook, parameterName=THreshold_Mouth}
  n33 = ScalarParameter() {defaultValue=1, group=MouthFlipbook, parameterName=Happy Mouth Frame}
  n34 = NamedRerouteUsage()
  n35 = ScalarParameter() {group=MouthFlipbook, parameterName=Blink_Mouth_Pose}
  n36 = NamedRerouteUsage()
  n37 = If(
   A=@n35
   B=@n32
   "A > B"=ScalarParameter() {defaultValue=2, group=MouthFlipbook, parameterName=Blink Mouth Frame}
   "A == B"=@n36
   "A < B"=@n36
  )
  n38 = ScalarParameter() {group=MouthFlipbook, parameterName=Sad_Mouth_Pose}
  n39 = NamedRerouteUsage()
  n40 = If(A=@n38, B=@n32, "A > B"=ScalarParameter() {group=MouthFlipbook, parameterName=Sad Mouth Frame}, "A == B"=@n39, "A < B"=@n39)
  n41 = Step(Y=NamedRerouteUsage(), X=@n38) {constY=0.1}
  n42 = NamedRerouteUsage()
  n43 = Max(A=@n38, B=@n35)
  NamedRerouteDeclaration(None=LinearInterpolate(
    A=If(A=@n31, B=@n32, "A > B"=@n33, "A == B"=@n34, "A < B"=@n34)
    B=LinearInterpolate(A=@n37, B=@n40, Alpha=@n41)
    Alpha=Step(Y=@n42, X=@n43) {constY=0.1}
   )) {name=MouthFrame, nodeColor=<rgba 1 0.2812 0 1>}
  NamedRerouteDeclaration(None=@n9.0-1 mask) {name=MouthUVMask, nodeColor=<rgba 0 1 0.6953 1>}
  n44 = TextureObjectParameter() {
   group=MouthFlipbook
   parameterName=MouthFlipbook
   texture=/Game/Accessories/FORT_Backpacks/Backpack_F_MED_SpaceFeline/Textures/T_RobotMouthFlipbook
  }
  n45 = ScalarParameter() {defaultValue=2, group=MouthFlipbook, parameterName=Mouth Sub Images Count}
  n46 = MaterialFunctionCall(Texture=@n44, UVs=NamedRerouteUsage(), SubImages=@n45, Frame=NamedRerouteUsage()) {materialFunction=&CustomNumbers_SubUV}
  NamedRerouteDeclaration(None=StaticSwitchParameter(True=NamedRerouteUsage(), False=@n46.RGB) {defaultValue=true, group=FlipBook, parameterName=UseFlipbookAsMask}) {name=Mouth Color, nodeColor=<rgba 0 1 0.8828 1>}
  n47 = TextureObjectParameter() {
   group=MouthFlipbook
   parameterName=MouthFlipbook_Mask
   samplerType=SAMPLERTYPE_Masks
   texture=/Game/Accessories/FORT_Backpacks/Backpack_F_MED_SpaceFeline/Textures/T_MouthFlipBook_Mask
  }
  n48 = MaterialFunctionCall(Texture=@n47, UVs=NamedRerouteUsage(), SubImages=@n45, Frame=NamedRerouteUsage()) {materialFunction=&CustomNumbers_SubUV}
  n49 = Multiply(
   A=StaticSwitchParameter(True=ComponentMask(None=@n48.RGB) {r=true}, False=@n46.Alpha) {defaultValue=true, group=FlipBook, parameterName=UseFlipbookAsMask}
   B=NamedRerouteUsage()
  )
  n50 = Step(Y=NamedRerouteUsage(), X=ScalarParameter() {group=MouthFlipbook, parameterName=Blank_Mouth_Pose}) {constY=0.1}
  NamedRerouteDeclaration(None=Multiply(A=NamedRerouteUsage(), B=LinearInterpolate(A=@n49, Alpha=@n50) {constB=0})) {name=Mouth Alpha, nodeColor=<rgba 0.3828 1 0 1>}
  NamedRerouteDeclaration(None=ChannelMaskParameter(None=FunctionInput() {inputName=Alpha, sortPriority=1, previewValue=<1 1 1 1>}) {group=EyeFlipbook, maskChannel=Green, parameterName=FlipBookChannel, defaultValue=<rgba 0 1 0 0>}.R) {name=Alpha IN, nodeColor=<rgba 0.3594 1 0 1>}
  NamedRerouteDeclaration(None=ScalarParameter() {defaultValue=3, group=EyeFlipbook, parameterName=Default Open Eye Netural Frame}) {name=DefaultEye, nodeColor=<rgba 1 0.5625 0 1>}
  n51 = ScalarParameter() {group=EyeFlipbook, parameterName=Eye UV Offset X}
  n52 = ScalarParameter() {group=EyeFlipbook, parameterName=Eye UV Offset Y}
  n53 = MaterialFunctionCall(
   UVs=Add(A=NamedRerouteUsage(), B=AppendVector(A=@n51, B=@n52))
   Texture Scale=ScalarParameter() {defaultValue=1, group=EyeFlipbook, parameterName=Eye UV Scale}
  ) {materialFunction=&ScaleUVsByCenter}
  NamedRerouteDeclaration(None=@n53.UVs) {name=Eye UV, nodeColor=<rgba 1 0 0.5625 1>}
  NamedRerouteDeclaration(None=@n53.0-1 mask) {name=EyeUVMask, nodeColor=<rgba 0 1 0.6953 1>}
 ]}
